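# Sound cue for a looping large waterfall ambience, played through the ambient water sound class with capsule waterfall attenuation.
# A quality-level root selects one of four modulated branches.
# The first picks randomly between two looping waves, while the others each loop a single wave, the last two using "_opt" variants.
# A concurrency override caps instances and stops the quietest, and the debug flag is left enabled.
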SoundCue /Game/Sounds/Fort_Ambient_Looping/Water/Waterfall_Large_A_02_Cue
root=QualityLevel_1
nodes [
 {id=Modulator_0, children [WavePlayer_0], properties {pitchMax=1.1, pitchMin=0.9}}
 {id=Modulator_1, children [WavePlayer_12], properties {pitchMax=1.1, pitchMin=0.9}}
 {id=Modulator_2, children [Random_2]}
 {id=Modulator_3, children [WavePlayer_1], properties {pitchMax=1.1, pitchMin=0.9}}
 {id=QualityLevel_1, children [Modulator_2,Modulator_1,Modulator_0,Modulator_3]}
 {id=Random_2, children [WavePlayer_10, WavePlayer_11], properties {weights [1,1]}}
 {id=WavePlayer_0, properties {bLooping=true, soundWaveAssetPtr=&/waterfall_new_ST_01_opt}}
 {id=WavePlayer_1, properties {bLooping=true, soundWaveAssetPtr=&waterfall_new_ST_01_opt}}
 {id=WavePlayer_10, properties {bLooping=true, soundWaveAssetPtr=&/waterfall_new_ST_01}}
 {id=WavePlayer_11, properties {bLooping=true, soundWaveAssetPtr=&/waterfall_new_ST_03}}
 {id=WavePlayer_12, properties {bLooping=true, soundWaveAssetPtr=&waterfall_new_ST_01}}
]
properties {
 attenuationSettings=/Game/Sounds/Attenuation/Fort_Ambient_Capsule_Waterfall_01_Att
 bDebug=true
 bOverrideConcurrency=true
 maxDistance=2227.1616
 soundClassObject=/Game/Sounds/FortSoundClassesAndModes/Fort_Ambient_Water
 volumeMultiplier=0.3624
 concurrencyOverrides {maxCount=4, resolutionRule=StopQuietest}
}
